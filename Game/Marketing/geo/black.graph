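# Plain black Default Lit material in the Marketing geo folder.
# Its graph holds only a color constant and a scalar constant, both left at zero, with no textures, parameters or math.
# Usage flags enable it on skeletal meshes and Niagara mesh particles.
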Material /Game/Marketing/geo/black
graph {
 properties {bUsedWithNiagaraMeshParticles=true, bUsedWithSkeletalMesh=true, pixelDepthOffsetMode=PDOM_Legacy, usedShadingModels=Default Lit}
 nodes [
  Constant3Vector()
  Constant()
 ]
}
