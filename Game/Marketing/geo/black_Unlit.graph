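# Unlit material in the Marketing geo folder, named as a flat black surface and enabled for skeletal mesh and mesh deformer use.
# The graph holds an RGB vector parameter defaulting to black and a Constant node, with no links from either to a material output shown.
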Material /Game/Marketing/geo/black_Unlit
graph {
 properties {bUsedWithMeshDeformer=true, bUsedWithSkeletalMesh=true, pixelDepthOffsetMode=PDOM_Legacy, shadingModel=MSM_Unlit, usedShadingModels=Unlit}
 nodes [
  Constant()
  VectorParameter() {parameterName=RGB, defaultValue=<rgba 0 0 0 1>}
 ]
}
