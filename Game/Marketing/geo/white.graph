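# A Default Lit material named white in the Marketing geo folder; the file does not show where it is used.
# The graph holds a white Constant3Vector and two scalar constants left at their default, with no textures or parameters.
# The listing shows no links from these nodes to the material outputs.
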Material /Game/Marketing/geo/white
graph {
 properties {pixelDepthOffsetMode=PDOM_Legacy, usedShadingModels=Default Lit}
 nodes [
  Constant3Vector() {constant=<rgba 1 1 1 1>}
  Constant()
  Constant()
 ]
}
